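; Projectile Blueprint for the SinisterMelon gameplay item that ignores its instigator when moving and splashes on entering its first water body.
; On authority, a lifetime timer stops projectile movement with a synthetic hit at the current location.
; The stop event stores the hit; authority clears the timer and plays the hit GameplayCue, then all sides disable collision, hide visuals, fade looping audio and kill it.
; An "Impact" debug print remains, plus a distance-travelled helper.
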(event EventBeginPlay
  (Variables|Default|SetStartLocation (Transformation|GetActorLocation))
  (Collision|IgnoreActorwhenMoving (Variables|Collision|GetCapsuleComponent) (Game|GetInstigator) true)
  (Networking|SwitchHasAuthority
    (:Authority
      (bind _returnvalue (Utilities|Time|SetTimerbyEvent (delegate Custom|MaxLifetimeElapsed) (ScalableFloat|GetValueatLevel (Variables|Default|GetRow_MaxLifetime))))
      (Variables|Default|SetLifetimeTimerHandle _returnvalue))))

(event Custom|MaxLifetimeElapsed
  (Game|Components|ProjectileMovement|StopSimulating (Variables|Movement|GetProjectileMovementComponent) (Collision|MakeHitResult false false 0.0 0.0 (Transformation|GetActorLocation) (Transformation|GetActorLocation) (Math|Rotator|GetRotationXVector (Transformation|GetActorRotation)) (Math|Rotator|GetRotationXVector (Transformation|GetActorRotation)))))

(event EventOnStop
  (params (Hit :type Hit%20Result%20Structure%20%28by%20ref%29))
  (Variables|Default|SetbHasStopped true)
  (Variables|Default|SetHitResult Hit)
  (Networking|SwitchHasAuthority
    (:Authority
      (Utilities|Time|ClearandInvalidateTimerbyHandle (Variables|Default|GetLifetimeTimerHandle))
      (CallFunction|PlayHitFX Hit))
    (:Remote
      _))
  (CallFunction|HideandKill))

(event Custom|PlayHitFX
  (params (HitResult :type Hit%20Result%20Structure%20%28by%20ref%29))
  (GameplayCue|ExecuteGameplayCue%28Burst%29Local self (Variables|Default|GetGC_Hit) (GameplayCue|MakeGameplayCueParametersfromHitResult HitResult))
  (Development|PrintString "Impact"))

(event Custom|HideAndKill
  (Collision|SetCollisionEnabled (Variables|Collision|GetCapsuleComponent) "NoCollision")
  (Rendering|SetVisibility (Variables|Default|GetStaticMesh) false false)
  (Components|Activation|Deactivate (Variables|Default|GetNiagara))
  (Delivery|Kill)
  (Audio|Components|Audio|FadeOut (Variables|Default|GetIn%20Air%20Looping%20Sound) 1.0 0.0 "Linear" :!class /Script/Engine.AudioComponent))

(event OnEnterWaterEvent(WaterInteractionComponent)
  (params (WaterBody :type Fort%20Water%20Body%20Actor%20Object%20Reference) (WaterInteractionComponent :type Fort%20Water%20Interaction%20Component%20Object%20Reference) (bIsFirstBody :type Boolean))
  (bind (_waterplanelocation _waterplanenormal) (FX|GetLastWaterSurfaceInfo WaterInteractionComponent :!class /Script/FortniteGame.FortWaterInteractionComponent))
  (if bIsFirstBody
    (Effects|Components|ParticleSystem|SpawnFXAtLocation (Variables|Default|GetWaterSplashFX) (Transformation|GetActorLocation) (Math|Rotator|MakeRotfromX _waterplanenormal))))

(fn DebugCheckDistanceTravelled
  (Development|PrintString (Utilities|String|ToString(Text) (Utilities|Text|FormatText "NSLOCTEXT(\"[21795E2624B5A7D23445EF8C4BDE08BE]\", \"B62F0FE840ED59F4BF47C68C076867C5\", \"Emerald Glass Projectile travelled: {Distance}\")" (/ (Math|Vector|Distance(Vector) (Variables|Default|GetStartLocation) (Transformation|GetActorLocation)) 1e+02)))))

(fn ConstructionScript)
