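; Segmented arc projectile for SinisterMelon, spawned by a sword swing, that damages and knocks back targets.
; Server collision checks disable hit segments, queue unique ability-system targets for delayed gameplay effects, and stop the projectile when all segments are gone.
; Replicated segment states drive Niagara visuals; a client predictive sphere trace triggers responsive impact cues.
; Water blocking can be delayed for swimming shooters, and cosmetic timers drive surface and wash-force effects.
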(event Custom|CleanUpFX
  (bind _fxwashforce (Variables|Default|GetFXWashForce))
  ;[0.4792 0.1725 0.1725 0.5] Add Physical_SurfaceFX
  (Utilities|IsValid _fxwashforce
    (:Is%20Valid
      (Actor|DestroyActor _fxwashforce))
    (:Is%20Not%20Valid
      _))
  (Components|Activation|Deactivate (Variables|Default|GetNS_SinisterMelon_Projectile_PhysicalSurfaceFX))
  (Utilities|Time|ClearandInvalidateTimerbyHandle (Variables|Default|GetFXPhysicalSurfaceTimer))
  (Components|DestroyComponent (Variables|Default|GetPostProcess)))

(event Custom|CE_Physical_SurfaceFX
  (bind _ns_sinistermelon_projectile_physicalsurfacefx (Variables|Default|GetNS_SinisterMelon_Projectile_PhysicalSurfaceFX))
  (bind _fxwashforce (Variables|Default|GetFXWashForce))
  ;[0.4792 0.1725 0.1725 0.5] Add Physical_SurfaceFX
  (bind (_outhit _returnvalue) (Collision|LineTraceForObjects (+ (Transformation|GetActorLocation) "0.000000,0.000000,200.000000") (- (Transformation|GetActorLocation) "0.000000,0.000000,300.000000") (Variables|Default|GetFXObject%20Types)))
  (if _returnvalue
    (if (not (.bSuccess (Utilities|Casting|CastToWaterBody (.HitActor (Collision|BreakHitResult _outhit)))))
      (Transformation|SetWorldLocation _ns_sinistermelon_projectile_physicalsurfacefx (.Location (Collision|BreakHitResult _outhit)) false false)
      (Utilities|IsValid _fxwashforce
        (:Is%20Valid
          (Variables|/Game/Effects/Niagara/Blueprints/PhsyicsFields/FN_WashForce.FN_WashForce_C|SetRadius 2e+03 _fxwashforce))
        (:Is%20Not%20Valid
          _))
      (Components|Activation|Activate _ns_sinistermelon_projectile_physicalsurfacefx false)))
  (Components|Activation|Deactivate _ns_sinistermelon_projectile_physicalsurfacefx)
  (Utilities|IsValid _fxwashforce
    (:Is%20Valid
      (Variables|/Game/Effects/Niagara/Blueprints/PhsyicsFields/FN_WashForce.FN_WashForce_C|SetRadius 1e+03 _fxwashforce))))

(event EventBeginPlay
  (bind (_returnvalue_location _returnvalue_rotation _returnvalue_scale) (Math|Transform|BreakTransform (Transformation|GetActorTransform)))
  (bind _ns_sinistermelon_prj (Variables|Default|GetNS_SinisterMelon_Prj))
  ;[0 0.1094 0.006088] Begin Play
  (|CallParentFunction)
  (Utilities|FlowControl|Sequence
    (:then_0
      ;;[1 1 1] Set FX in code
      (Niagara|ReinitializeSystem _ns_sinistermelon_prj)
      ;;[1 1 1] Make sure the FX is fully visible
      (Components|Activation|Activate _ns_sinistermelon_prj true)
      (CallFunction|UpdateVisuals)
      (bind _asfort_player_pawn (Utilities|Casting|CastToFortPlayerPawn (Variables|Actor|GetInstigator)))
      (Variables|Transient|SetWeapon (Variables|/Script/FortniteGame.FortPawn|GetCurrentWeapon _asfort_player_pawn))
      ;;[0 0.15 0.1456 0.5] Cache predictive impact trace start location.
      (bind _returnvalue (TargetingSystem|GetPlayerCameraTargetingTransform _asfort_player_pawn))
      (Variables|Transient|SetPreviousClientPredictiveImpactTraceStartLocation (.location _returnvalue)))
    (:then_1
      (CallFunction|BPInitializeSegments)
      (Networking|SwitchHasAuthority
        (:Authority
          (bind _shoulddelay (CallFunction|ShouldDelayBlockingWater))
          (if _shoulddelay
            ;;[1 1 1] Init Water Blocking
            (Utilities|Time|SetTimerbyEvent (delegate Custom|DelayedBlockHitsWithWater) (ScalableFloat|GetValueatLevel (Variables|WaterBlocking|GetBlockWaterDelay))) ; Delay blocking water for a small period of time to let the projectile travel out of the water
            (else
              (Variables|/Script/FortniteGame.FortProjectileSegmentedCollisionComponent|SetbBlockHitsWithWater true (Variables|Default|GetFortProjectileSegmentedCollision)))))
        (:Remote
          (if (not (Variables|Transient|GetbProjectileStopped))
            (Sound|PlayinAirSound)))))
    (:then_2
      (Variables|Projectile|SetMaxSimulationDistanceSquaredOverride (Math|Float|Square (ScalableFloat|GetValueatLevel (Variables|Hotfix|GetHotfix_MaxRange))))
      (CallFunction|CollisionCheck))
    (:then_3
      ;[0.4792 0.1725 0.1725 0.5] Add Physical_SurfaceFX
      ; Add FX force
      (Networking|CanExecuteCosmeticEvents
        (:True
          (Utilities|FlowControl|Sequence
            (:then_0
              (bind _returnvalue_2 (Development|GetConsoleVariableIntValue "r.PhysicsField.EnableField"))
              (if (== _returnvalue_2 1)
                (bind _returnvalue_3 (Game|SpawnActorfromClass "/Game/Effects/Niagara/Blueprints/PhsyicsFields/FN_WashForce.FN_WashForce_C" (Math|Transform|MakeTransform (- _returnvalue_location "0.000000,200.000000,0.000000") _returnvalue_rotation _returnvalue_scale) "Undefined" "MultiplyWithRoot" 0 2e+03 1.3 -999.0 0))
                (bind _output_get (Variables|Default|SetFXWashForce _returnvalue_3))
                (Transformation|AttachActorToActor _output_get self "None" "SnapToTarget" "KeepRelative" "KeepRelative" true)))
            (:then_1
              (bind _returnvalue_4 (Utilities|Time|SetTimerbyEvent (delegate Custom|CE_Physical_SurfaceFX) 0.3 true))
              (Variables|Default|SetFXPhysicalSurfaceTimer _returnvalue_4)
              (Transformation|DetachFromComponent (Variables|Default|GetNS_SinisterMelon_Projectile_PhysicalSurfaceFX) "KeepRelative" "KeepRelative" "KeepRelative" true))))))))

(event Custom|DelayedBlockHitsWithWater
  ;[0 0.1094 0.006088] Begin Play
  ;;[1 1 1] Init Water Blocking
  (Variables|/Script/FortniteGame.FortProjectileSegmentedCollisionComponent|SetbBlockHitsWithWater true (Variables|Default|GetFortProjectileSegmentedCollision)))

(event EventTick
  (params (DeltaSeconds :type Float%20%28single-precision%29))
  ;[0.1249 1 0.7414] Tick
  (|CallParentFunction DeltaSeconds)
  (Utilities|FlowControl|Sequence
    (:then_0
      (CallFunction|CollisionCheck)
      (CallFunction|ProcessHits false))
    (:then_1
      (Networking|CanExecuteCosmeticEvents
        (:True
          (Impact|RunClientPredictiveImpactTrace))))))

(event Custom|CollisionCheck
  ;[0.4549 0.9373 1] Collision Check
  (if (Networking|IsDedicatedServer)
    (bind _returnvalue (Class|FortProjectileSegmentedCollisionComponent|TickCollisionCheck (Variables|Default|GetFortProjectileSegmentedCollision)))
    (for _array_element _returnvalue
      (bind (_array_element_segmentindex _array_element_hitresult) (Utilities|Struct|BreakCollisionSegmentHit _array_element))
      (if (Team|OnSameTeam (Variables|Actor|GetInstigator) (.HitActor (Collision|BreakHitResult _array_element_hitresult)))
        _
        (elif (Variables|/Script/Engine.Actor|GetbHidden (.HitActor (Collision|BreakHitResult _array_element_hitresult)))
          _
          (else
            (CallFunction|OnHitActor (.HitActor (Collision|BreakHitResult _array_element_hitresult)) _array_element_hitresult)
            (if (Collision|BreakHitResult _array_element_hitresult)
              (Class|FortProjectileSegmentedCollisionComponent|EnableCollisionSegment (Variables|Default|GetFortProjectileSegmentedCollision) _array_element_segmentindex false)
              (Utilities|Array|SetArrayElem (Variables|Transient|GetSegmentEnabledStatusReplicated) _array_element_segmentindex) ; Update replicated array
              (if (< (Math|Vector|DistanceSquared(Vector) (.ImpactPoint (Collision|BreakHitResult _array_element_hitresult)) (Transformation|GetActorLocation)) (Math|Float|Square (ScalableFloat|GetCachedScalableFloatValue (Variables|Hotfix|GetHotfix_MidpointImpactDetectDistanceThreshold))))
                (Utilities|FlowControl|DoOnce false
                  (:Completed
                    (CallFunction|PlayProjectileImpactFX _array_element_hitresult)))))))))
    (if (Utilities|Array|IsNotEmpty _returnvalue)
      ;;[1 0.4428 0.1199] Maybe Start Death
      (bind (_output_segmentindex _output_hitresult) (Utilities|Struct|BreakCollisionSegmentHit (Utilities|Array|Get(aref) _returnvalue)))
      (CallFunction|StartDeathifAllSegmentsDisabled _output_hitresult)
      ;;[1 0.6285 0] If there was a hit this check, ForceNetUpdate to get the info to the client asap.
      ;; FORT-718860 - Helps make segment kills in niagara more reliable/on frame
      (Networking|ForceNetUpdate))))

(event Custom|OnHitActor
  (params (Actor :type Actor%20Object%20Reference) (Hit :type Hit%20Result%20Structure))
  (bind _returnvalue (Ability|GetFortAbilitySystemComponent Actor))
  (bind (_bblockinghit _binitialoverlap) (Collision|BreakHitResult Hit))
  (bind _alreadyhitascs (Variables|Transient|GetAlreadyHitASCs))
  ;[1 0.7289 0.4762] OnHitActor:
  ; This is also called from the mantis GA when the larger melee distance trace is done when the projectile is spawned (Sword is swung)
  (Utilities|FlowControl|Sequence
    (:then_0
      (Utilities|IsValid _returnvalue
        (:Is%20Valid
          (if (not (Utilities|Array|ContainsItem _alreadyhitascs _returnvalue))
            (Utilities|Array|Add _alreadyhitascs _returnvalue) ; Store unique hit ASCs instead of Actors\n so we don't double hit the pawn and an attached actor \n(eg the pawn and the olympus wings)
            (Utilities|Array|Add (Variables|Transient|GetPendingHits) (Utilities|Struct|MakeSinisterMelonProjectileHit Hit (Utilities|Time|GetTimeSeconds))) ; Hits should always be in time order.
          ))))
    (:then_1
      (if (not _binitialoverlap)
        (CallFunction|OnHitActorFX Hit)))
    (:then_2
      (Development|PrintString "OnHitActor" true true "(R=1.000000,G=0.044170,B=0.000000,A=1.000000)" 22.0))))

(event Custom|FXSpawnSegmentDeathFX
  (bind (_returnvalue_location _returnvalue_rotation _returnvalue_scale) (Math|Transform|BreakTransform (Transformation|GetWorldTransform (Variables|Default|GetNS_SinisterMelon_Prj))))
  ;[1 0.3866 0 0.533] VFX - Spawns death fx for each segment, on projectile death
  (Rendering|SetVisibility (Variables|Default|GetNS_SinisterMelon_Prj) false false)
  (if (> (CallFunction|GetNumSegmentsEnabled) 0)
    (bind _returnvalue (Niagara|SpawnSystematLocation (Variables|FX|GetNS_OnDeath) _returnvalue_location _returnvalue_rotation _returnvalue_scale true false "AutoRelease"))
    (VFX|FXSetSegmentParams (Variables|Default|GetNS_SinisterMelon_Prj) _returnvalue)
    (Components|Activation|Activate _returnvalue false)
    (CallFunction|CleanUpFX)))

(event Custom|OnHitActor_FX
  (params (Hit :type Hit%20Result%20Structure%20%28by%20ref%29))
  ;[0.349 0.002116 0.3241 0.5] OnHit FX (Multicast)
  (Networking|CanExecuteCosmeticEvents
    (:True
      (FX|HandleSegmentImpactFXWater Hit)
      (FX|HandleSegmentImpactFXPlayer Hit)
      (FX|HandleSegmentImpactFXBuilding Hit)
      (FX|HandleSegmentImpactFXLandscape Hit)
      (CallFunction|CleanUpFX))))

(event Custom|PlayProjectileImpactFX
  (params (ImpactHitResult :type Hit%20Result%20Structure))
  ;[0 0.15 0.1112 0.5] PlayProjectileImpactFX: This is for a single large explosion where the arc's mid point hits. This should spawn decal, etc. either through the gameplay cue or directly in this logic.
  (Networking|CanExecuteCosmeticEvents
    (:True
      (if (not (Variables|Transient|GetbImpactFXComplete))
        (Variables|Transient|SetbImpactFXComplete true)
        (GameplayCues|ExecuteGameplayCue (Variables|FX|GetImpactTag))
        (CallFunction|CleanUpFX)))))

(event Custom|StopInAirAudio
  (params (FadeOutDuration :type Float%20%28single-precision%29))
  ;[0.1124 0.4685 1] Audio
  (bind _inairloopaudio (Variables|Sound|GetInAirLoopAudio)
    (:then
      (Audio|Components|Audio|FadeOut _inairloopaudio FadeOutDuration 0.0 "Linear" :!class /Script/Engine.AudioComponent))
    (:else)))

(event EventEndPlay
  (params (EndPlayReason :type EEndPlayReason%20Enum :default "Destroyed"))
  ;[1 0.7299 0.2392] EndPlay
  (|CallParentFunction EndPlayReason)
  (Networking|SwitchHasAuthority
    (:Remote
      (CallFunction|StopinAirAudio 0.25)
      (Utilities|Time|ClearandInvalidateTimerbyHandle (Variables|Default|GetFXPhysicalSurfaceTimer))
      (Components|Activation|Deactivate (Variables|Default|GetNS_SinisterMelon_Projectile_PhysicalSurfaceFX)))))

(event EventOnStop
  (params (Hit :type Hit%20Result%20Structure%20%28by%20ref%29))
  ;[1 0.7299 0.2392] OnStop
  (|CallParentFunction Hit)
  (Actor|Tick|SetActorTickEnabled false) ; We disable tick on the server in StartDeathifAllSegmentsDisabled. \nStop it on the client here
  (Networking|SwitchHasAuthority
    (:Authority
      (Variables|Transient|SetbProjectileStopped true)
      (CallFunction|ProcessHits true) ; Finish processing any pending hits
    )
    (:Remote
      (CallFunction|FXSpawnSegmentDeathFX))))

(fn ApplyHitResultToEffectContext
  (params (InEffectContextHandle :type Gameplay%20Effect%20Context%20Handle%20Structure) (InHitResult :type Hit%20Result%20Structure))
  (bind _output_get (LocalVariables|SetEffectContextHandle InEffectContextHandle))
  (Ability|EffectContext|AddHitResult _output_get InHitResult)
  (return (LocalVariables|GetEffectContextHandle)))

(fn BP_InitializeSegments
  (for _ (Variables|/Script/FortniteGame.FortProjectileSegmentedCollisionComponent|GetCollisionSegments (Variables|Default|GetFortProjectileSegmentedCollision))
    (Utilities|Array|Add (Variables|Transient|GetSegmentEnabledStatusReplicated) "true")))

(fn FXGetSegmentParams
  (return 6 (Math|Conversions|ToVector(Float) 3.0) 59))

(fn FXGetSegmentTransforms
  (return (Utilities|Array|MakeArray (Math|Transform|MakeTransform "-104.000000,-264.000000,0.000000" "0.000000,-40.000000,0.000000") (Math|Transform|MakeTransform "-41.000000,-168.000000,0.000000" "0.000000,-26.000000,0.000000") (Math|Transform|MakeTransform "-5.000000,-58.000000,0.000000" "0.000000,-10.000000,0.000000") (Math|Transform|MakeTransform "-5.000000,58.000000,0.000000" "0.000000,10.000000,0.000000") (Math|Transform|MakeTransform "-41.000000,168.000000,0.000000" "0.000000,26.000000,0.000000") (Math|Transform|MakeTransform "-104.000000,264.000000,0.000000" "0.000000,40.000000,0.000000"))))

(fn FXSetSegmentParams
  (params (SourceSystem :type Niagara%20Particle%20System%20Component%20Object%20Reference) (NiagaraSystem :type Niagara%20Particle%20System%20Component%20Object%20Reference))
  (bind _returnvalue (Utilities|Name|MakeLiteralName "SegmentPositions"))
  (bind _returnvalue_2 (Utilities|Name|MakeLiteralName "SegmentRotations"))
  (bind _niagarasystem (Variables|GetNiagaraSystem))
  (Utilities|FlowControl|Sequence
    (:then_0
      (bind _returnvalue_3 (Niagara|NiagaraGetQuaternionArray (Variables|GetSourceSystem) _returnvalue_2))
      (Niagara|NiagaraSetQuaternionArray (Variables|GetNiagaraSystem) _returnvalue_2 _returnvalue_3))
    (:then_1
      (bind _returnvalue_4 (Niagara|NiagaraGetVectorArray (Variables|GetSourceSystem) _returnvalue))
      (Niagara|NiagaraSetVectorArray (Variables|GetNiagaraSystem) _returnvalue _returnvalue_4))
    (:then_2
      (Niagara|NiagaraSetBoolArray (Variables|GetNiagaraSystem) "ActiveArray" (Variables|Transient|GetSegmentEnabledStatusReplicated)))
    (:then_3
      (Niagara|SetNiagaraVariable(Int32) _niagarasystem "NumSegments" (VFX|FXGetSegmentParams))
      (Niagara|SetNiagaraVariable(Int32) _niagarasystem "SegmentWidth" 0))
    (:then_4
      (Niagara|SetNiagaraVariable(Vector3) (Variables|GetNiagaraSystem) "RootScale" "1.000000,1.000000,1.000000"))))

(fn GetNumSegmentsEnabled
  (for _array_element (Variables|Transient|GetSegmentEnabledStatusReplicated)
    (if _array_element
      (Math|Integer|IncrementInt (LocalVariables|GetLocalNumSegmentsEnabled))))
  (return (LocalVariables|GetLocalNumSegmentsEnabled)))

(fn HandleSegmentImpactFX_Building
  (params (HitResult :type Hit%20Result%20Structure))
  (bind (_bblockinghit _binitialoverlap _time _distance _location _impactpoint _normal _impactnormal _physmat _hitactor _hitcomponent) (Collision|BreakHitResult HitResult))
  (bind (_validcheck _canbedamaged) (Class|BPFLAICommonFunctions|GetCanbeDamaged (Components|GetOwner _hitcomponent)))
  (Networking|CanExecuteCosmeticEvents
    (:True
      (if _canbedamaged
        (bind _returnvalue (Niagara|SpawnSystematLocation :SystemTemplate (Variables|FX|GetNS_Impact_Building) :Location _impactpoint :Scale "1.000000,1.000000,1.000000" :bAutoDestroy true :bAutoActivate true :PoolingMethod "AutoRelease" :bPreCullCheck true))
        (Niagara|SetNiagaraVariable(Float) _returnvalue "GlobalScale" (Variables|FX|GetFXImpactScale))
        (Niagara|SetNiagaraVariable(Vector3) _returnvalue "Direction" (Transformation|GetActorRightVector))))))

(fn HandleSegmentImpactFX_Landscape
  (params (HitResult :type Hit%20Result%20Structure))
  (bind (_bblockinghit _binitialoverlap _time _distance _location _impactpoint _normal _impactnormal _physmat _hitactor _hitcomponent) (Collision|BreakHitResult HitResult))
  (bind (_aslandscape _bsuccess) (Utilities|Casting|CastToLandscape (Components|GetOwner _hitcomponent)))
  (bind (_aslandscape_proxy _bsuccess_2) (Utilities|Casting|CastToLandscapeProxy (Components|GetOwner _hitcomponent)))
  (bind (_aslandscape_spline_actor _bsuccess_3) (Utilities|Casting|CastToLandscapeSplineActor (Components|GetOwner _hitcomponent)))
  (bind (_aslandscape_spline_meshes_actor _bsuccess_4) (Utilities|Casting|CastToLandscapeSplineMeshesActor (Components|GetOwner _hitcomponent)))
  (Networking|CanExecuteCosmeticEvents
    (:True
      (if (Math|Boolean|ORBoolean _bsuccess _bsuccess_2 _bsuccess_4 _bsuccess_3)
        _))))

(fn HandleSegmentImpactFX_Player
  (params (HitResult :type Hit%20Result%20Structure))
  (bind (_bblockinghit _binitialoverlap _time _distance _location _impactpoint _normal _impactnormal _physmat _hitactor _hitcomponent) (Collision|BreakHitResult HitResult))
  (bind (_asfort_pawn _bsuccess) (Utilities|Casting|CastToFortPawn (Components|GetOwner _hitcomponent)))
  (bind (_asfort_aipawn _bsuccess_2) (Utilities|Casting|CastToFortAIPawn (Components|GetOwner _hitcomponent)))
  (Networking|CanExecuteCosmeticEvents
    (:True
      (if (or _bsuccess _bsuccess_2)
        (bind _returnvalue (Niagara|SpawnSystematLocation :SystemTemplate (Variables|FX|GetNS_Impact_Player) :Location _impactpoint :Scale "1.000000,1.000000,1.000000" :bAutoDestroy true :bAutoActivate true :PoolingMethod "AutoRelease" :bPreCullCheck true))
        (Niagara|SetNiagaraVariable(Float) _returnvalue "GlobalScale" (Variables|FX|GetFXImpactScale))))))

(fn HandleSegmentImpactFX_Water
  (params (HitResult :type Hit%20Result%20Structure))
  (bind (_bblockinghit _binitialoverlap _time _distance _location _impactpoint _normal _impactnormal _physmat _hitactor _hitcomponent) (Collision|BreakHitResult HitResult))
  (Networking|CanExecuteCosmeticEvents
    (:True
      (if (Utilities|Enum|Equal(Enum) (Collision|GetCollisionObjectType _hitcomponent) "ECC_GameTraceChannel18")
        (bind _returnvalue (Niagara|SpawnSystematLocation "/Game/Effects/Niagara/Systems/Enviroment/Water/NS_WaterSplash_Medium.NS_WaterSplash_Medium" _impactpoint (Math|Conversions|RotationFromXVector _impactnormal) "1.000000,1.000000,1.000000" true true "AutoRelease"))
        (Niagara|SetNiagaraVariable(Float) _returnvalue "GlobalScale" (Variables|FX|GetFXImpactScale))
        (Audio|PlaySoundatLocation (Variables|Sound|GetHitWater) _impactpoint)))))

(fn OnRep_bProjectileStopped
  (if (Variables|Transient|GetbProjectileStopped)
    (CallFunction|StopinAirAudio 0.25)
    (Audio|PlaySoundatLocation (Variables|Sound|GetStopSound) (Transformation|GetActorLocation))))

(fn OnRep_Collsion)

(fn OnRep_SegmentEnabledStatusReplicated
  (Networking|CanExecuteCosmeticEvents
    (:True
      (CallFunction|UpdateVisuals))))

(fn PlayInAirSound
  (Variables|Sound|GetInAirLoopAudio
    (:then)
    (:else
      (bind _capsulecomponent (Variables|Collision|GetCapsuleComponent)
        (:then
          (bind _returnvalue (Audio|SpawnSoundAttached (Variables|Sound|GetInAirLoopSound) _capsulecomponent))
          (Variables|Sound|SetInAirLoopAudio _returnvalue))
        (:else)))))

(fn ProcessHits
  (params (bProcessAll :type Boolean))
  (bind _instigator (Variables|Actor|GetInstigator))
  (bind _returnvalue (Ability|GetFortAbilitySystemComponent _instigator))
  (bind _output_get (LocalVariables|SetMaxIterations (Utilities|Array|Length (Variables|Transient|GetPendingHits))))
  (for _ (range _output_get)
    (bind _output_get_2 (LocalVariables|SetCurrentHitData (Utilities|Array|Get(aref) (Variables|Transient|GetPendingHits))))
    (bind (_output_get_hit_4_eb4813be4dbc4b4b1bd455af284a0a37 _output_get_worldtime_8_a2e1c5a649bf06c0059ba4aabe18305a) (Utilities|Struct|BreakSinisterMelonProjectileHit _output_get_2))
    (if (or (Variables|GetbProcessAll) (>= (- (Utilities|Time|GetTimeSeconds) _output_get_worldtime_8_a2e1c5a649bf06c0059ba4aabe18305a) (ScalableFloat|GetCachedScalableFloatValue (Variables|Hotfix|GetHotfix_HitDelay))))
      (Utilities|Array|RemoveIndex (Variables|Transient|GetPendingHits))
      (bind _outeffectcontexthandle (CallFunction|ApplyHitResulttoEffectContext (GameplayEffects|MakeFortEffectContext _returnvalue _instigator _instigator (Variables|Transient|GetWeapon) 0) _output_get_hit_4_eb4813be4dbc4b4b1bd455af284a0a37))
      (Utilities|IsValid (Ability|GetFortAbilitySystemComponent (.HitActor (Collision|BreakHitResult _output_get_hit_4_eb4813be4dbc4b4b1bd455af284a0a37)))
        (:Is%20Valid
          (GameplayEffects|ApplyGameplayEffectToTarget _returnvalue (Variables|Settings|GetDamageGameplayEffectClass) (Ability|GetFortAbilitySystemComponent (.HitActor (Collision|BreakHitResult _output_get_hit_4_eb4813be4dbc4b4b1bd455af284a0a37))) 0.0 _outeffectcontexthandle)
          (GameplayEffects|ApplyGameplayEffectToTarget _returnvalue (Variables|Settings|GetKnockbackGameplayEffectClass) (Ability|GetFortAbilitySystemComponent (.HitActor (Collision|BreakHitResult _output_get_hit_4_eb4813be4dbc4b4b1bd455af284a0a37))) 0.0 _outeffectcontexthandle)))
      (else
        (return)))))

(fn RunClientPredictiveImpactTrace
  (bind _actorstoignore (LocalVariables|GetActorsToIgnore))
  ;[0.00079 0.09375 0.01769 0.5] RunClientPredictiveImpactTrace: Run a simple trace to figure out where the projectile is going to hit and play the impact fx there for responsiveness.
  (Networking|CanExecuteCosmeticEvents
    (:True
      (if (> (ScalableFloat|GetCachedScalableFloatValue (Variables|Hotfix|GetHotfix_RunClientPredictiveImpactTrace)) 0)
        (bind _instigator (Variables|Actor|GetInstigator)
          (:then
            (if (Pawn|IsLocallyViewed _instigator)
              (Utilities|Array|AppendArray _actorstoignore (Utilities|Array|MakeArray (Variables|Actor|GetInstigator)))
              (Utilities|Array|AppendArray _actorstoignore (Class|FortPawn|GetProjectileIgnoredActors (Utilities|Casting|CastToFortPlayerPawnAthena (Variables|Actor|GetInstigator))))
              (Utilities|FlowControl|Sequence
                (:then_0
                  (if (not (Variables|Transient|GetbImpactFXComplete))
                    (bind _outhit (Collision|SphereTraceByChannel (Variables|Transient|GetPreviousClientPredictiveImpactTraceStartLocation) (Transformation|GetActorLocation) (* (Variables|Transient|GetCurrentGrowthScale) 1e+01) "TraceTypeQuery6" false (LocalVariables|GetActorsToIgnore)))
                    (if (Collision|BreakHitResult _outhit)
                      (CallFunction|PlayProjectileImpactFX _outhit))))
                (:then_1
                  (Variables|Transient|SetPreviousClientPredictiveImpactTraceStartLocation (Transformation|GetActorLocation))))))
          (:else))))))

(fn SetHeavyHitMeleeTargetHitResults
  (params (HitResults :type Array%20of%20Hit%20Result%20Structures))
  ;[0.15 0 0.08959 0.5] SetHeavyHitMeleeTargetHitResults: Called from the mantis GA. This caches the hit results from the trace done to represent the melee swing.
  (Networking|SwitchHasAuthority
    (:Authority
      (bind _output_get (Variables|Transient|SetHeavyHitMeleeTargetHitResults HitResults))
      (for _array_element _output_get
        (bind _returnvalue (GameplayTags|HasMatchingGameplayTag(Message) (.HitActor (Collision|BreakHitResult _array_element)) "(TagName=\"Athena.Ability.TargetFilter.ConsiderAsPawn\")"))
        (if (or _returnvalue (Utilities|IsA (.HitActor (Collision|BreakHitResult _array_element)) "/Script/FortniteGame.FortPawn"))
          (Utilities|Array|Add (LocalVariables|GetActorsToIgnore) (.HitActor (Collision|BreakHitResult _array_element)))))
      (Class|FortProjectileSegmentedCollisionComponent|SetIgnoredActors (Variables|Default|GetFortProjectileSegmentedCollision) (LocalVariables|GetActorsToIgnore)))))

(fn ShouldDelayBlockingWater
  (if (> (ScalableFloat|GetValueatLevel (Variables|WaterBlocking|GetBlockWaterDelay)) 0)
    (bind _asfort_player_pawn (Utilities|Casting|CastToFortPlayerPawn (Variables|Actor|GetInstigator)))
    (if (WaterBody|IsSurfaceSwimming _asfort_player_pawn)
      (if (ScalableFloat|GetValueasBool (Variables|WaterBlocking|GetBlockWater_UseReticleAboveWaterTest))
        (bind _weapon (Variables|Transient|GetWeapon)
          (:then
            ;[0.5191 0.7095 1] Test if the reticle is above the water line
            (FortWeapon|UpdateDamageStartLocation _weapon) ; No way to tell if the cached damage start location is up to date, so update
            (bind (_outhit _returnvalue) (Collision|LineTraceForObjects :Start (.ReturnValue (FortWeapon|GetDamageStartLocation _weapon -1)) :End (+ (.ReturnValue (FortWeapon|GetDamageStartLocation _weapon -1)) (* (FortWeapon|GetDamageStartLocation _weapon -1) (ScalableFloat|GetValueatLevel (Variables|WaterBlocking|GetBlockWater_ReticleAboveWaterTraceDist)))) :ObjectTypes (Utilities|Array|MakeArray "ObjectTypeQuery13") :bTraceComplex false :DrawDebugType (select (LocalVariables|GetbDebug) "ForDuration" "None") :bIgnoreSelf true :TraceColor "(R=1.000000,G=0.000000,B=0.000000,A=1.000000)" :TraceHitColor "(R=0.000000,G=1.000000,B=0.000000,A=1.000000)" :DrawTime 5.0)) ; Trace just for water to determine if the reticle is aimed above the water line
            ;[1 0.6235 0.2] If we hit water, the reticle is aimed below the water line. Block water immediately (no delay).
            ; Otherwise delay to give the projectile a chance to spawn and move out of the water
            (Development|PrintString (Utilities|String|SelectString "Reticle above water line, delaying blocking water" "Reticle below water line, block water immediately" (not _returnvalue)) (LocalVariables|GetbDebug) (LocalVariables|GetbDebug))
            (return (not _returnvalue)))
          (:else
            _))
        (else
          ;[1 0.6235 0.2] Delay for all swiming cases
          (Development|PrintString "UseReticleAboveWaterTest is false, delaying blocking water for all swimming cases" (LocalVariables|GetbDebug) (LocalVariables|GetbDebug))
          (return true)))
      (else
        _))
    ;[1 0.6235 0.2] No Delay
    (return false)
    (else
      (Development|PrintString "BlockWaterDelay is <=0, blocking water immediately" (LocalVariables|GetbDebug) (LocalVariables|GetbDebug)))))

(fn StartDeathIfAllSegmentsDisabled
  (params (HitThisTick :type Hit%20Result%20Structure))
  (bind (_bblockinghit _binitialoverlap _time _distance _location) (Collision|BreakHitResult HitThisTick))
  (Networking|SwitchHasAuthority
    (:Authority
      (LocalVariables|SetShouldStartDeath true)
      (bind _array_element (Utilities|Array|ForEachLoopwithBreak (Variables|Transient|GetSegmentEnabledStatusReplicated))
        (:LoopBody
          (if _array_element
            (LocalVariables|SetShouldStartDeath false)))
        (:Completed
          _))
      (if (LocalVariables|GetShouldStartDeath)
        (Actor|SetLifeSpan (Variables|Delivery|GetLifespanAfterKill))
        (Game|Components|ProjectileMovement|StopSimulating (Variables|Movement|GetProjectileMovementComponent) (Collision|MakeHitResult false false 0.0 0.0 (+ (Transformation|GetActorLocation) (Math|Vector|ProjectVectorontoVector (- _location (Transformation|GetActorLocation)) (Transformation|GetActorForwardVector)))))
        (Actor|Tick|SetActorTickEnabled false)
        (CallFunction|ProcessHits true) ; Finish processing any pending hits
      ))))

(fn UpdateVisuals
  (Networking|CanExecuteCosmeticEvents
    (:True
      (for (_array_element _array_index) (Variables|Transient|GetSegmentEnabledStatusReplicated)
        (Niagara|NiagaraSetBoolArrayValue (Variables|Default|GetNS_SinisterMelon_Prj) "ActiveArray" _array_index _array_element true)))))

(fn ConstructionScript)
